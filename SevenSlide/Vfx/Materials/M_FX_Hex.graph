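# M_FX_Hex is an unlit, additive material for Niagara sprite effects that draws a particle-tinted hex shape.
# The red and green channels of the hex texture are scaled by Base and Ring parameters, summed, and multiplied by particle color.
# The result passes through the shared MF_FX_EmissiveAdjustments function.
# Particle alpha is multiplied by a depth fade with an adjustable distance to soften intersections with geometry.
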Material /SevenSlide/Vfx/Materials/M_FX_Hex
graph {
 properties {
  bAutomaticallySetUsageInEditor=false
  bUsedWithNiagaraSprites=true
  blendMode=BLEND_Additive
  shadingModel=MSM_Unlit
  usedShadingModels=Unlit
 }
 nodes [
  n1 = TextureSample() {samplerType=SAMPLERTYPE_LinearColor, texture=/SevenSlide/Vfx/Textures/T_FX_Hex}
  n2 = Multiply(A=@n1.R, B=ScalarParameter() {defaultValue=0.15, parameterName=Base})
  n3 = Multiply(A=@n1.G, B=ScalarParameter() {defaultValue=3, parameterName=Ring})
  n4 = ParticleColor()
  MaterialFunctionCall(Emissive=Multiply(A=Add(A=@n2, B=@n3), B=@n4.RGB)) {materialFunction=&MF_FX_EmissiveAdjustments}
  Multiply(A=@n4.A, B=DepthFade(FadeDistance=ScalarParameter() {defaultValue=10, parameterName=FadeDistance}))
 ]
}
