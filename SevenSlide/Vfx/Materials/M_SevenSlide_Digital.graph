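# Translucent unlit, two-sided Niagara material for sprite, ribbon and mesh particles, producing a tinted glow effect.
# A depth-faded FXCommon_SoftOutline result drives both brightness and opacity.
# Emissive blends a color parameter with particle color through ParticleColorOnly, then passes through MF_FX_EmissiveAdjustments.
# Opacity combines that mask with particle alpha, an inverted Fresnel that fades edges, and an optional sine-over-time flicker.
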Material /SevenSlide/Vfx/Materials/M_SevenSlide_Digital
graph {
 properties {
  bAutomaticallySetUsageInEditor=false
  bUsedWithNiagaraMeshParticles=true
  bUsedWithNiagaraRibbons=true
  bUsedWithNiagaraSprites=true
  blendMode=BLEND_Translucent
  opacityMaskClipValue=0.1
  shadingModel=MSM_Unlit
  twoSided=true
  usedShadingModels=Unlit
 }
 nodes [
  n1 = ParticleColor()
  n2 = DepthFade(
   Opacity=MaterialFunctionCall!(
    GlowMeshStencilIndex=ScalarParameter() {defaultValue=40, parameterName="GlowMeshStencilIndex (S)"}
    UseCustomDepth=StaticBoolParameter() {parameterName=UseCustomDepth}
    UseCustomStencil=StaticBoolParameter() {parameterName=UseCustomStencil}
    OutlineThickness=ScalarParameter() {defaultValue=2.7, parameterName="OutlineThickness (S)"}
    OutlineStrength
    PixelDepthOffset
    MaxBlurDepth=ScalarParameter() {defaultValue=150, parameterName="MaxBlurDepth (S)"}
    ------EnableInnerGlow------=ScalarParameter() {defaultValue=1}
    InnerGlowMaxArea=ScalarParameter() {defaultValue=100, parameterName="InnerGlowMaxArea (S)"}
    InnerGlowPower=ScalarParameter() {defaultValue=10, parameterName="InnerGlowPower (S)"}
    InnerGlowStrength=ScalarParameter() {defaultValue=0.5, parameterName="InnerGlowStrength (S)"}
    ------LessUse------
    MeshPosition
    Distance Steps
    Radial Steps
   ) {materialFunction=/Game/Effects/Fort_Effects/Effects/Atlas/Materials/FXCommon_SoftOutline}.Result
   FadeDistance=ScalarParameter() {defaultValue=100, parameterName=DepthFade}
  )
  n3 = VectorParameter() {parameterName=C, defaultValue=<rgba 0.7182 7 0.5126 0.08>}
  n4 = Multiply(A=@n2, B=@n3.A)
  n5 = ParticleColor()
  n6 = LinearInterpolate(A=Multiply(A=@n3.RGB, B=@n5.RGB), B=@n5.RGB, Alpha=ScalarParameter() {parameterName=ParticleColorOnly})
  MaterialFunctionCall(Emissive=Add(A=Multiply(A=@n1.RGB, B=@n4), B=Divide(A=@n6))) {materialFunction=&MF_FX_EmissiveAdjustments}
  n7 = Multiply(A=LinearInterpolate(A=@n4, B=@n2, Alpha=ScalarParameter() {parameterName=ParticleColorOnly}), B=@n5.A)
  n8 = Fresnel(ExponentIn=ScalarParameter() {defaultValue=3, parameterName=Fresnel_Exponent}) {baseReflectFraction=0, exponent=1}
  n9 = Saturate(None=Power(Base=OneMinus(None=@n8), Exp=ScalarParameter() {parameterName=Fresnel_Power}))
  n10 = Time()
  n11 = ScalarParameter() {defaultValue=10, parameterName=Flick_Speed}
  n12 = MaterialFunctionCall(
   Input=Sine(None=Multiply(A=@n10, B=@n11))
   Input Low=Constant() {r=-1}
   Input High=Constant() {r=1}
   Target Low=ScalarParameter() {defaultValue=0.7, parameterName=Flick_Low}
   Target High=ScalarParameter() {defaultValue=1.3, parameterName=Flick_High}
  ) {materialFunction=&RemapValueRange}
  n13 = ScalarParameter() {parameterName=EnableFlick}
  Multiply(A=@n2, B=Multiply(A=Multiply(A=@n7, B=@n9), B=LinearInterpolate(B=@n12.Result, Alpha=@n13) {constA=1}))
 ]
}
