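# M_Trail_SevenSlide is a translucent, unlit, two-sided material for Niagara ribbon trails in SevenSlide.
# Its graph builds a soft procedural mask without textures, multiplying parabolic x(1-x) falloffs along UV1 U and UV0 V to fade toward the edges.
# Particle alpha scales the mask, which feeds a DepthFade with a parameterized distance to soften intersections with scene geometry.
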Material /SevenSlide/Vfx/Materials/M_Trail_SevenSlide
graph {
 properties {
  bAutomaticallySetUsageInEditor=false
  bUsedWithNiagaraRibbons=true
  blendMode=BLEND_Translucent
  shadingModel=MSM_Unlit
  twoSided=true
  usedShadingModels=Unlit
 }
 nodes [
  n1 = MaterialFunctionCall(Float2=TextureCoordinate() {coordinateIndex=1}) {materialFunction=&BreakOutFloat2Components}
  n2 = Multiply(A=Multiply(A=@n1.R, B=OneMinus(None=@n1.R))) {constB=2}
  n3 = MaterialFunctionCall(Float2=TextureCoordinate()) {materialFunction=&BreakOutFloat2Components}
  n4 = OneMinus(None=@n3.G)
  n5 = Saturate(None=Multiply(A=Multiply(A=@n3.G, B=@n4)) {constB=2})
  DepthFade(Opacity=Multiply(A=ParticleColor().A, B=Multiply(A=@n2, B=@n5)), FadeDistance=ScalarParameter() {defaultValue=100, parameterName=DepthFade})
 ]
}
