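# Post-process material, blended after depth of field, that adds a glowing hexagon overlay toward the screen edges.
# Aspect-corrected screen UVs sample a hex mask, feeding sine-animated tile pulses, noise breakup and tinted emissive.
# An inverted, powered sphere mask confines the glow to the periphery; Low quality, Forward and Mobile use only this tinted vignette.
# BlendWeight lerps inversely: 0 shows the full effect, while the default 1 leaves the scene untouched.
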Material /SevenSlide/Vfx/Materials/PPM_SevenSliders
graph {
 properties {
  blendableLocation=BL_SceneColorAfterDOF
  materialDomain=MD_PostProcess
  pixelDepthOffsetMode=PDOM_Legacy
  shadingModel=MSM_Unlit
  usedShadingModels=Unlit
 }
 nodes [
  n1 = NamedRerouteUsage()
  n2 = Time()
  n3 = Subtract(
   A=NamedRerouteUsage()
   B=Multiply(A=Multiply(A=@n1, B=@n2), B=ScalarParameter() {defaultValue=0.0001, group=SL_S30_Hexagon, parameterName=MaskDisplacementSpeed})
  )
  n4 = Saturate(None=Multiply(A=Sine(None=@n3), B=ScalarParameter() {defaultValue=0.3, group=SL_S30_Hexagon, parameterName=HexTileDepthColor}) {constB=0.1})
  n5 = ScalarParameter() {defaultValue=1, group=MyGroup, parameterName=HexOverallIntensity}
  n6 = NamedRerouteUsage()
  n7 = Step(Y=@n6) {constX=0.25}
  n8 = NamedRerouteUsage()
  n9 = Multiply(
   A=@n6
   B=SmoothStep(
    Max=OneMinus(None=@n7)
    Value=TextureSampleParameter2D(UVs=@n8) {parameterName=Noise, texture=/SevenSlide/Vfx/Textures/T_FoundKelp_Pattern_Seamless}.G
   )
  )
  n10 = Multiply(A=NamedRerouteUsage(), B=Multiply(A=Multiply(A=@n4, B=@n5), B=Multiply(A=@n9) {constB=2})) {constB=5}
  n11 = TextureCoordinate()
  n12 = Constant2Vector() {g=0.5, r=0.5}
  n13 = Constant() {r=0.6}
  n14 = Constant()
  n15 = Power(Base=OneMinus(None=SphereMask(A=@n11, B=@n12, Radius=@n13, Hardness=@n14))) {constExponent=5}
  n16 = Multiply(A=@n15, B=VectorParameter() {group=Derez, parameterName=Color_03, defaultValue=<rgba 0.1946 0.7204 0 1>}.RGB)
  n17 = ShadingPathSwitch(Default=Multiply(A=Multiply(A=@n10, B=@n15), B=@n15), Forward=@n16, Mobile=@n16)
  n18 = QualitySwitch(Default=@n17, Low=@n16, High=@n17, Medium=@n17, Epic=@n17)
  n19 = ComponentMask(None=SceneTexture() {sceneTextureId=PPI_PostProcessInput0}.Color) {b=true, g=true, r=true}
  LinearInterpolate(A=Add(A=Multiply(A=@n18) {constB=2}, B=@n19), B=@n19, Alpha=ScalarParameter() {defaultValue=1, parameterName=BlendWeight})
  n20 = Add(A=ScreenPosition().ViewportUV) {constB=-0.5}
  n21 = ViewSize()
  n22 = Multiply(
   A=Divide(A=Multiply(A=@n20, B=@n21), B=ComponentMask(None=@n21) {g=true})
   B=ScalarParameter() {defaultValue=3.33, group=SL_S30_Hexagon, parameterName=Hex_Tiling}
  ) {constB=5}
  n23 = ComponentMask(None=VectorParameter() {group=SL_S30_Hexagon, parameterName=Hex_TilingRatio, defaultValue=<rgba 0.85 1 0 1>}.RGB) {g=true, r=true}
  n24 = TextureSample(UVs=NamedRerouteDeclaration(None=Multiply(A=@n22, B=@n23)) {name=Hex_UV, nodeColor=<rgba 0 1 0.4609 1>}) {
   samplerSource=SSM_Wrap_WorldGroupSettings
   texture=/Game/Global/Materials/MaterialGym/Materials/HexDisplacement/Textures/T_Hexagon_Displacement_Mask_V2
  }
  NamedRerouteDeclaration(None=@n24.G) {name=Mask, nodeColor=<rgba 0.0547 0 1 1>}
  NamedRerouteDeclaration(None=@n24.R) {name=HexPattern_Greyscale, nodeColor=<rgba 0.125 1 0 1>}
  n25 = LinearInterpolate(
   A=VectorParameter() {group=Derez, parameterName=Color_01, defaultValue=<rgba 0.054 0.2 0 1>}.RGB
   B=VectorParameter() {group=Derez, parameterName=Color_02, defaultValue=<rgba 0 5 0.1475 1>}.RGB
   Alpha=NamedRerouteUsage()
  )
  NamedRerouteDeclaration(None=LinearInterpolate(A=@n25, B=Multiply(A=@n25) {constB=5}) {constAlpha=0}) {name=Emissive, nodeColor=<rgba 0 0.3438 1 1>}
  n26 = NamedRerouteUsage()
  n27 = NamedRerouteDeclaration(None=Multiply(A=Time(), B=ScalarParameter() {defaultValue=0.3, group=SL_S30_Hexagon, parameterName=HexIdleDisplacementSpeed})) {name=Time, nodeColor=<rgba 0.9219 0 1 1>}
  n28 = Multiply(A=Sine(None=Subtract(A=@n26, B=@n27)), B=ScalarParameter() {defaultValue=0.3, group=SL_S30_Hexagon, parameterName=HexTileDepthColor}) {constB=0.1}
  NamedRerouteDeclaration(None=Multiply(A=Saturate(None=@n28), B=ScalarParameter() {defaultValue=1, group=MyGroup, parameterName=HexOverallIntensity})) {name=AnimatedHexHash, nodeColor=<rgba 0.5938 0 1 1>}
 ]
}
