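# Niagara system for SevenSlide engine thrusters, placing exhaust effects on four skeletal mesh sockets or bones through particle partitioning.
# It layers burst flash sprites, mesh reactor and flame, flickering flares with lights, and GPU sparks.
# User.Fuel scales flame and reactor size and light intensity, with age curves adding a sputtering ignition ramp.
# System Timeline blends flame, reactor, light and spark colors from orange to blue.
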NiagaraSystem /SevenSlide/Vfx/NS_SevenSlide_EngineThruster
dependencies {
 usedDataInterfaces [
  &NiagaraDataInterfaceCurve
  &NiagaraDataInterfaceColorCurve
  &NiagaraDataInterfaceMeshRendererInfo
  &NiagaraDataInterfaceArrayFloat3
  &NiagaraDataInterfaceArrayFloat
  &NiagaraDataInterfaceSkeletalMesh
 ]
 usedDynamicInputs []
 usedModules [
  &self:SystemSpawnScript.SystemScriptSource.SystemScriptGraph.NiagaraNodeAssignment_0.SetVariables_#1
  &SystemState
  &Timeline
 ]
}
emitters {
 FlameBig {
  emitterUpdateScript [
   &EmitterState()
   &SpawnRate(SpawnRate=25, Use Spawn Probability=false)
  ]
  particleSpawnScript [
   &PartitionParticles(Number Of Partitions=4)
   &InitializeParticle(Color=<rgba 1 1 1 0.5>, Mesh Scale=<0.25 0.05 0.1>, Mesh Scale Mode=::NewEnumerator4, UsePositionOffset=false, Write Material Random=true)
   &InitialMeshOrientation(
    Mesh Axis To Orient=<-1 0 0>
    Mesh Orientation Mode=::NewEnumerator2
    Orientation Coordinate Space=::Local
    Rotation=&MergeHelper_VectorSwitch(0=<0 0 0>, 1=<0 0 0>, 2=<0 0 0>, 3=<0 0 0>, Index=$Particles.PartitionParticles.Partition, Num=::NewEnumerator2)
    Use Rotation Vector=true
   )
   &CameraOffset(Camera Offset Amount=5)
  ]
  particleUpdateScript [
   &ParticleState()
   /SetVariables(Particles.FuelCurve:NiagaraFloat=&FloatFromCurve(
     FloatCurve=DataInterface() {propertyValues {Curve {
        defaultValue=0
        keys [
         {interpMode=RCIM_Cubic, leaveTangentWeight=0.6109, tangentMode=RCTM_User, time=-0.0016, value=0.25}
         {arriveTangent=0.1618, interpMode=RCIM_Cubic, leaveTangent=0.1618, time=0.0628, value=0.3108}
         {arriveTangent=0.8253, interpMode=RCIM_Cubic, leaveTangent=0.8253, time=0.1209, value=0.2698}
         {arriveTangent=0.1419, interpMode=RCIM_Cubic, leaveTangent=0.1419, time=0.2017, value=0.4254}
         {arriveTangent=0.6867, interpMode=RCIM_Cubic, leaveTangent=0.6867, time=0.2978, value=0.2949}
         {arriveTangent=0.2529, interpMode=RCIM_Cubic, leaveTangent=0.2529, time=0.4036, value=0.564}
         {arriveTangent=0.1725, interpMode=RCIM_Cubic, leaveTangent=0.1725, time=0.5363, value=0.3552}
         {arriveTangent=0.0085, interpMode=RCIM_Cubic, leaveTangent=0.0085, time=0.6709, value=0.6101}
         {arriveTangent=0.5479, interpMode=RCIM_Cubic, leaveTangent=0.5479, time=0.7752, value=0.3573}
         {arriveTangent=0.3976, interpMode=RCIM_Cubic, leaveTangent=0.3976, time=0.8739, value=0.7214}
         {arriveTangent=0.2807, interpMode=RCIM_Cubic, leaveTangent=0.2807, time=1.0222, value=0.4554}
         {arriveTangent=0.1076, interpMode=RCIM_Cubic, leaveTangent=0.1076, time=1.1478, value=0.7982}
         {arriveTangent=0.4978, interpMode=RCIM_Cubic, leaveTangent=0.4978, time=1.2858, value=0.4838}
         {arriveTangent=0.3971, interpMode=RCIM_Cubic, leaveTangent=0.3971, time=1.5246, value=0.9858}
         {arriveTangent=0.0463, interpMode=RCIM_Cubic, leaveTangent=0.0463, time=1.6797, value=0.6402}
         {arriveTangent=2.0988, interpMode=RCIM_Cubic, leaveTangent=2.0988, tangentMode=RCTM_User, time=1.831, value=1}
        ]
       }}}
     CurveIndex=$Emitter.Age
     Scale Curve=&FloatFromCurve(
      FloatCurve=DataInterface() {propertyValues {Curve {
         defaultValue=0
         keys [
          {arriveTangent=27.1748, interpMode=RCIM_Cubic, leaveTangent=27.1748, tangentMode=RCTM_User, time=0, value=0.25}
          {arriveTangent=2.9597, interpMode=RCIM_Cubic, leaveTangent=2.9597, time=0.0416, value=0.5938}
          {arriveTangent=1.1179, interpMode=RCIM_Cubic, leaveTangent=1.1179, tangentMode=RCTM_User, time=0.1922, value=0.8189}
          {arriveTangent=0.0224, interpMode=RCIM_Cubic, leaveTangent=0.0224, tangentMode=RCTM_User, time=1, value=1}
         ]
        }}}
      CurveIndex=$User.Fuel
     )
    )) {isSetParameters=true}
   &SkeletalMeshLocation(
    Bone / Socket Index=$Particles.PartitionParticles.Partition
    MaskSpawnGroup=false
    Mesh Sampling Type=::NewEnumerator5
    Sampled Position Offset=&MergeHelper_VectorSwitch(0=<0 -1.5 0.5>, 1=<0 -1.5 0.5>, 2=<0 0 0>, 3=<0 0 0>, Index=$Particles.PartitionParticles.Partition, Num=::NewEnumerator2)
    Skeletal Mesh=$System.Skeletal Mesh
    Socket Sampling Mode=::NewEnumerator5
   )
   &ScaleMeshSize(Scale Factor=&Multiply_VectorByFloat(
     Vector=&SelectVectorFromArray(
      Direct Array Index=$Particles.SpawnGroup
      Vector Selection Array=DataInterface() {propertyValues {FloatData [<1 1.7 1.7>, <1 1.7 1.7>, <1 1.7 1.7>, <1 1.7 1.7>]}}
     )
     Float=&FloatFromCurve(
      FloatCurve=DataInterface() {propertyValues {Curve {keys [<at 0 1>,<at 1 1>]}}}
      Scale Curve=&RandomRangeFloat(Minimum=0.7, Maximum=1, Evaluation Type=::NewEnumerator1)
     )
    ))
   &ScaleMeshSize(Scale Factor=&MakeVector(X=$Particles.FuelCurve, Y=1, Z=1)) {module=ScaleMeshSize002}
   &ScaleMeshSize(Scale Factor=&Multiply_VectorByFloat(
     Vector=<1 1 1>
     Float=&SelectFloatFromArray(Direct Array Index=$Particles.SpawnGroup, Float Selection Array=DataInterface() {propertyValues {FloatData [0.8,0.8,1,1]}})
    )) {module=ScaleMeshSize001}
   &ScaleColor(
    Scale Alpha=&FloatFromCurve(
     FloatCurve=DataInterface() {propertyValues {Curve {keys [
         {interpMode=RCIM_Cubic, tangentMode=RCTM_User, time=0, value=0}
         {interpMode=RCIM_Cubic, tangentMode=RCTM_User, time=0.5, value=1}
         {tangentMode=RCTM_User, time=1, value=0}
        ]}}}
     Scale Curve=0.5
    )
    Scale RGB=&MakeVectorFromLinearColorRGB(Color=&LerpLinearColors(
      StartColor=&ColorFromCurve(ColorCurve=DataInterface() {propertyValues {
         AlphaCurve {keys []}
         BlueCurve {keys [{arriveTangent=-7.5103, interpMode=RCIM_Cubic, time=0, value=0.3516}, <at 1 0.0469>]}
         GreenCurve {keys [{arriveTangent=-2.0408, interpMode=RCIM_Cubic, time=0, value=1.3461}, <at 1 0.0469>]}
         RedCurve {keys [{arriveTangent=-1.9216, interpMode=RCIM_Cubic, time=0, value=7.5}, <at 1 1>]}
        }})
      EndColor=&ColorFromCurve(ColorCurve=DataInterface() {propertyValues {
         AlphaCurve {keys []}
         BlueCurve {keys [{arriveTangent=-7.5103, interpMode=RCIM_Cubic, time=0, value=7.5}, <at 1 1>]}
         GreenCurve {keys [{arriveTangent=-2.0408, interpMode=RCIM_Cubic, time=0, value=3.9258}, <at 1 0.0469>]}
         RedCurve {keys [{arriveTangent=-1.9216, interpMode=RCIM_Cubic, time=0, value=0.3516}, <at 1 0.0469>]}
        }})
      LerpFactor=$System.Timeline.NormalizedTime
     ))
    ScaleA=true
    ScaleRGB=true
   )
  ]
  rendererClasses [&NiagaraMeshRendererProperties]
  renderers [{rendererClass=&NiagaraMeshRendererProperties, rendererIndex=0}]
 }
 Flash {
  emitterUpdateScript [
   &EmitterState(Inactive Response=::NewEnumerator1, Life Cycle Mode=::NewEnumerator1, Loop Behavior=::NewEnumerator1, Loop Duration=2, UseLoopDelay=false)
   &SpawnBurst_Instantaneous(Spawn Count=4, Use Spawn Probability=false, UseLoopCountLimit=false)
  ]
  particleSpawnScript [
   &PartitionParticles(Number Of Partitions=4)
   &InitializeParticle(
    Color=<rgba 1 0.125 0 1>
    Lifetime=1.5
    Sprite Size Mode=::NewEnumerator3
    Uniform Sprite Size=50
    UsePositionOffset=false
    Write Material Random=true
   )
   &CameraOffset(Camera Offset Amount=10)
  ]
  particleUpdateScript [
   &ParticleState()
   &SkeletalMeshLocation(
    Bone / Socket Index=$Particles.PartitionParticles.Partition
    "Interpolate Mesh Sampling (Update)"=true
    MaskSpawnGroup=false
    Mesh Sampling Type=::NewEnumerator5
    Skeletal Mesh=$System.Skeletal Mesh
    Socket Sampling Mode=::NewEnumerator5
   )
   &ScaleSpriteSize(
    Uniform Curve Scale=&RandomRangeFloat(Minimum=0.5, Maximum=1.5, Evaluation Type=::NewEnumerator1)
    Uniform Curve Sprite Scale=DataInterface() {propertyValues {Curve {keys [
        {arriveTangent=1.8088, interpMode=RCIM_Cubic, leaveTangent=1.8088, tangentMode=RCTM_User, time=0, value=0}
        {arriveTangent=3, interpMode=RCIM_Cubic, leaveTangent=3, time=0.3545, value=1.3786}
        {arriveTangent=1.2403, tangentMode=RCTM_User, time=1, value=3}
       ]}}}
   )
   &DynamicMaterialParameters(
    Index 0 Param 1=1
    Index 0 Param 2=1
    Module.Index0Float0Write=true
    Module.Index0Float1Write=true
    Module.Index0Float2Write=true
    Module.Index0Float3Write=false
   )
   &ScaleColor(
    Linear Color Curve=DataInterface() {propertyValues {
      AlphaCurve {keys [<at 0.054 0>,<at 0.1199 1>,<at 0.939 0>]}
      BlueCurve {keys [<at 0 2>,<at 0.2321 1>,<at 1 1>]}
      GreenCurve {keys [<at 0 2>,<at 0.2321 1>,<at 1 1>]}
      RedCurve {keys [<at 0 2>,<at 0.2321 1>,<at 1 1>]}
     }}
    Scale Mode=2
   )
  ]
  rendererClasses [&NiagaraSpriteRendererProperties]
  renderers [{rendererClass=&NiagaraSpriteRendererProperties}]
 }
 Light {
  emitterUpdateScript [
   &EmitterState(
    Inactive Response=::NewEnumerator1
    Life Cycle Mode=::NewEnumerator1
    Loop Behavior=::NewEnumerator1
    Loop Duration Mode=::NewEnumerator1
    UseLoopDelay=false
   )
   /SetVariables(Emitter.FlareColor:LinearColor=&LerpLinearColors(StartColor=<rgba 20 1.3499 0 0.1>, EndColor=<rgba 0 5 20 0.1>, LerpFactor=$System.Timeline.NormalizedTime)) {isSetParameters=true}
   &SpawnBurst_Instantaneous(Spawn Count=4, Use Spawn Probability=false, UseLoopCountLimit=false)
   &SpawnBurst_Instantaneous(Spawn Count=2, Spawn Group=1, Use Spawn Probability=false, UseLoopCountLimit=false) {module=SpawnBurst_Instantaneous001}
  ]
  particleSpawnScript [
   /SetVariables(Particles.VisibilityTag=&SpawnGroup()) {isSetParameters=true}
   &PartitionParticles(Number Of Partitions=4)
   &InitializeParticle(
    Sprite Size Mode=::NewEnumerator3
    Uniform Sprite Size=&MergeHelper_FloatSwitch(0=75, 1=75, 2=50, 3=50, 4=0, Index=$Particles.PartitionParticles.Partition, Num=::NewEnumerator3)
    UsePositionOffset=false
    Write Material Random=true
   )
   &SystemLocation(MaskSpawnGroup=false, Offset=$System.PSystemPosition)
   &CameraOffset(Camera Offset Amount=&MergeHelper_FloatSwitch(0=2.5, 1=2.5, 2=2.5, 3=2.5, Index=$Particles.PartitionParticles.Partition, Num=::NewEnumerator2))
  ]
  particleUpdateScript [
   &ParticleState(Kill Particles When Lifetime Has Elapsed=false, Let Infinitely Lived Particles Die When Emitter Deactivates=true)
   &SkeletalMeshLocation(
    Bone / Socket Index=$Particles.PartitionParticles.Partition
    MaskSpawnGroup=false
    Mesh Sampling Type=::NewEnumerator5
    Sampled Position Offset=<4 0 0>
    Skeletal Mesh=$System.Skeletal Mesh
    Socket Sampling Mode=::NewEnumerator5
   )
   &ScaleColor(
    Scale Alpha=&Waveform(
     Global Amplitude Scale=5
     "Waveform [1]"=::NewEnumerator12
     "[1] Frequency"=5
     "[1] Amplitude Min/Max"=<0.5 1>
     "[1] Scale"=&Waveform("Waveform [1]"=::NewEnumerator12, "[1] Frequency"=2, "[1] Amplitude Min/Max"=<0.5 1>)
    )
    Scale RGB=&MakeVectorFromLinearColorRGB(Color=&LerpLinearColors(StartColor=<rgba 2 0.1667 0 1>, EndColor=<rgba 0 0.5 2 1>, LerpFactor=$System.Timeline.NormalizedTime))
    ScaleA=true
    ScaleRGB=true
   )
   &ScaleColor(
    Scale Alpha=&FloatFromCurve(
     FloatCurve=DataInterface() {propertyValues {Curve {
        defaultValue=0
        keys [
         {arriveTangent=4.1785, interpMode=RCIM_Cubic, leaveTangent=4.1785, tangentMode=RCTM_User, time=0, value=0}
         {arriveTangent=1.1179, interpMode=RCIM_Cubic, leaveTangent=1.1179, tangentMode=RCTM_User, time=0.2042, value=0.6228}
         {arriveTangent=0.0224, interpMode=RCIM_Cubic, leaveTangent=0.0224, tangentMode=RCTM_User, time=1, value=1}
        ]
       }}}
     CurveIndex=$User.Fuel
    )
    ScaleA=true
    ScaleRGB=true
   ) {module=ScaleColor001}
   &DynamicMaterialParameters(
    Index 0 Param 1=10
    Index 0 Param 2=0.5
    Module.Index0Param0Write=true
    Module.Index0Param1Write=true
    Module.Index0Param2Write=false
    Module.Index0Param3Write=false
   )
   &FN_Light_Attributes(
    Color=$Particles.Color
    Create Color Attribute=true
    Create Exponent Attribute=true
    Create Radius Attribute=true
    Exponent=5
    Radius=50
    Scale Color=&FloatFromCurve(
     FloatCurve=DataInterface() {propertyValues {Curve {
        defaultValue=0
        keys [
         {arriveTangent=4.1785, interpMode=RCIM_Cubic, leaveTangent=4.1785, tangentMode=RCTM_User, time=0, value=0}
         {arriveTangent=1.1179, interpMode=RCIM_Cubic, leaveTangent=1.1179, tangentMode=RCTM_User, time=0.2042, value=0.6228}
         {arriveTangent=0.0224, interpMode=RCIM_Cubic, leaveTangent=0.0224, tangentMode=RCTM_User, time=1, value=1}
        ]
       }}}
     CurveIndex=$User.Fuel
     Scale Curve=7.5
    )
   )
   &FN_Sprite_Attributes(
    Camera Offset=-15
    Color=&LerpLinearColors(StartColor=<rgba 5 0.8333 0 2.2>, EndColor=<rgba 0 1.0833 5 2.2>, LerpFactor=$System.Timeline.NormalizedTime)
    Create Camera Offset Attribute=true
    Create Color Attribute=true
    Create Dynamic Material Parameter Attributes=true
    Create SpriteSize Attribute=true
    Index 0 Param 1=1
    Index 0 Param 2=10
    Index 0 Param 4=0.15
    Module Output:ENiagara_AttributeSamplingApplyOutput=::NewEnumerator1
    Module.Index0Float0Write=true
    Module.Index0Float1Write=true
    Module.Index0Float2Write=true
    Module.Index0Float3Write=true
    SpriteSize=&MergeHelper_Vector2Switch(0=<10 1>, 1=<10 1>, 2=<1 10>, 3=<1 10>, 4=?, Index=$Particles.PartitionParticles.Partition, Num=::NewEnumerator2)
   )
  ]
  rendererClasses [&NiagaraSpriteRendererProperties, &NiagaraLightRendererProperties]
  renderers [
   {rendererClass=&NiagaraSpriteRendererProperties}
   {rendererClass=&NiagaraLightRendererProperties, rendererIndex=1}
   {rendererClass=&NiagaraSpriteRendererProperties, rendererIndex=2}
  ]
 }
 Reactor {
  emitterUpdateScript [
   &EmitterState(
    Inactive Response=::NewEnumerator1
    Life Cycle Mode=::NewEnumerator1
    Loop Behavior=::NewEnumerator1
    Loop Delay=0.25
    Loop Duration Mode=::NewEnumerator1
    UseLoopDelay=true
   )
   &SpawnBurst_Instantaneous(Spawn Count=4, Use Spawn Probability=false, UseLoopCountLimit=false)
  ]
  particleSpawnScript [
   &PartitionParticles(Number Of Partitions=4)
   &InitializeParticle(
    Lifetime=0.5
    Mesh Renderer Array Visibility Mode=::NewEnumerator1
    Mesh Renderer Info=DataInterface() {propertyValues {MeshRenderer=&self:Reactor_0.NiagaraMeshRendererProperties_0}}
    Mesh Renderer Mesh Index:NiagaraInt32=&MergeHelper_IntSwitch(0=1, 1=1, 2=0, 3=0, 4=?, 5=?, Index=$Particles.PartitionParticles.Partition, Num=::NewEnumerator2)
    Mesh Scale=&Multiply_VectorByFloat(
     Vector=&MergeHelper_VectorSwitch(0=<1 1 3.5>, 1=<1 1 3.5>, 2=<1 3.5 1.2>, 3=<1 3.5 1.2>, Index=$Particles.PartitionParticles.Partition, Num=::NewEnumerator2)
     Float=1
    )
    Mesh Scale Mode=::NewEnumerator4
    UsePositionOffset=false
    Write Material Random=true
   )
   &InitialMeshOrientation(
    Mesh Axis To Orient=<-1 0 0>
    Mesh Orientation Mode=::NewEnumerator2
    Orientation Coordinate Space=::Local
    Rotation=&MergeHelper_VectorSwitch(0=<0 0 0>, 1=<0 0 0>, 2=<0 0 0>, 3=<0 0 0>, Index=$Particles.PartitionParticles.Partition, Num=::NewEnumerator2)
    Use Rotation Vector=true
   )
  ]
  particleUpdateScript [
   &ParticleState(Kill Particles When Lifetime Has Elapsed=false, Let Infinitely Lived Particles Die When Emitter Deactivates=true)
   /SetVariables(Particles.FuelCurve:NiagaraFloat=&FloatFromCurve(
     FloatCurve=DataInterface() {propertyValues {Curve {
        defaultValue=0
        keys [
         {interpMode=RCIM_Cubic, leaveTangentWeight=0.6109, tangentMode=RCTM_User, time=-0.0016, value=0}
         {arriveTangent=0.1618, interpMode=RCIM_Cubic, leaveTangent=0.1618, time=0.0628, value=0.0608}
         {arriveTangent=0.8253, interpMode=RCIM_Cubic, leaveTangent=0.8253, time=0.1209, value=0.0198}
         {arriveTangent=0.1419, interpMode=RCIM_Cubic, leaveTangent=0.1419, time=0.2017, value=0.1754}
         {arriveTangent=0.6867, interpMode=RCIM_Cubic, leaveTangent=0.6867, time=0.2978, value=0.0449}
         {arriveTangent=0.2529, interpMode=RCIM_Cubic, leaveTangent=0.2529, time=0.4036, value=0.314}
         {arriveTangent=0.1725, interpMode=RCIM_Cubic, leaveTangent=0.1725, time=0.5363, value=0.1052}
         {arriveTangent=0.0085, interpMode=RCIM_Cubic, leaveTangent=0.0085, time=0.6709, value=0.3601}
         {arriveTangent=0.5479, interpMode=RCIM_Cubic, leaveTangent=0.5479, time=0.7752, value=0.1073}
         {arriveTangent=0.3976, interpMode=RCIM_Cubic, leaveTangent=0.3976, time=0.8739, value=0.4714}
         {arriveTangent=0.2807, interpMode=RCIM_Cubic, leaveTangent=0.2807, time=1.0222, value=0.2054}
         {arriveTangent=0.1076, interpMode=RCIM_Cubic, leaveTangent=0.1076, time=1.1478, value=0.5482}
         {arriveTangent=1.1612, interpMode=RCIM_Cubic, leaveTangent=1.1612, time=1.2858, value=0.2338}
         {arriveTangent=1.0317, interpMode=RCIM_Cubic, leaveTangent=1.0317, time=1.5246, value=0.9858}
         {arriveTangent=0.0463, interpMode=RCIM_Cubic, leaveTangent=0.0463, time=1.6797, value=0.6402}
         {arriveTangent=2.0988, interpMode=RCIM_Cubic, leaveTangent=2.0988, tangentMode=RCTM_User, time=1.831, value=1}
        ]
       }}}
     CurveIndex=$Emitter.Age
     Scale Curve=&FloatFromCurve(
      FloatCurve=DataInterface() {propertyValues {Curve {
         defaultValue=0
         keys [
          {arriveTangent=27.1748, interpMode=RCIM_Cubic, leaveTangent=27.1748, tangentMode=RCTM_User, time=0, value=0}
          {arriveTangent=4.2603, interpMode=RCIM_Cubic, leaveTangent=4.2603, time=0.0385, value=0.5297}
          {arriveTangent=1.1179, interpMode=RCIM_Cubic, leaveTangent=1.1179, tangentMode=RCTM_User, time=0.1922, value=0.8189}
          {arriveTangent=0.0224, interpMode=RCIM_Cubic, leaveTangent=0.0224, tangentMode=RCTM_User, time=1, value=1}
         ]
        }}}
      CurveIndex=$User.Fuel
     )
    )) {isSetParameters=true}
   &SkeletalMeshLocation(
    Bone / Socket Index=$Particles.PartitionParticles.Partition
    MaskSpawnGroup=false
    Mesh Sampling Type=::NewEnumerator5
    Sampled Position Offset=&MergeHelper_VectorSwitch(0=<0 -1.5 0.5>, 1=<0 -1.5 0.5>, 2=<0 0 0>, 3=<0 0 0>, Index=$Particles.PartitionParticles.Partition, Num=::NewEnumerator2)
    Skeletal Mesh=$System.Skeletal Mesh
    Socket Sampling Mode=::NewEnumerator5
   )
   &ScaleMeshSize(Scale Factor=&MergeHelper_VectorSwitch(
     0=&MakeVector(X=1, Y=1, Z=$Particles.FuelCurve)
     1=&MakeVector(X=1, Y=1, Z=$Particles.FuelCurve)
     2=&MakeVector(X=1, Y=$Particles.FuelCurve, Z=1)
     3=&MakeVector(X=1, Y=$Particles.FuelCurve, Z=1)
     Index=$Particles.PartitionParticles.Partition
     Num=::NewEnumerator2
    )) {module=ScaleMeshSize001}
   &ScaleColor(
    Scale Alpha=&FloatFromCurve(
     FloatCurve=DataInterface() {propertyValues {Curve {
        defaultValue=0
        keys [
         {arriveTangent=27.1748, interpMode=RCIM_Cubic, leaveTangent=27.1748, tangentMode=RCTM_User, time=0, value=0}
         {arriveTangent=4.2603, interpMode=RCIM_Cubic, leaveTangent=4.2603, time=0.0385, value=0.5297}
         {arriveTangent=1.1179, interpMode=RCIM_Cubic, leaveTangent=1.1179, tangentMode=RCTM_User, time=0.1922, value=0.8189}
         {arriveTangent=0.0224, interpMode=RCIM_Cubic, leaveTangent=0.0224, tangentMode=RCTM_User, time=1, value=1}
        ]
       }}}
     CurveIndex=$User.Fuel
    )
    Scale RGB=&MakeVectorFromLinearColorRGB(Color=&LerpLinearColors(StartColor=<rgba 5 0.5833 0 1>, EndColor=<rgba 0 0.8333 5 1>, LerpFactor=$System.Timeline.NormalizedTime))
    ScaleA=false
    ScaleRGB=true
   ) {module=ScaleColor001}
   &ScaleColor(
    Linear Color Curve=DataInterface() {propertyValues {
      AlphaCurve {keys [<at 0.054 0>,<at 0.1199 0.5>]}
      BlueCurve {keys [<at 0 2>,<at 0.2321 1>,<at 1 1>]}
      GreenCurve {keys [<at 0 2>,<at 0.2321 1>,<at 1 1>]}
      RedCurve {keys [<at 0 2>,<at 0.2321 1>,<at 1 1>]}
     }}
    Scale Mode=2
   )
  ]
  rendererClasses [&NiagaraMeshRendererProperties]
  renderers [{rendererClass=&NiagaraMeshRendererProperties, rendererIndex=0}]
 }
 Sparks001 {
  simTarget=GPUComputeSim
  emitterUpdateScript [
   &EmitterState(
    Inactive Response=::NewEnumerator1
    Life Cycle Mode=::NewEnumerator1
    Loop Behavior=::NewEnumerator1
    Loop Duration Mode=::NewEnumerator1
    UseLoopDelay=false
   )
   &SpawnRate(SpawnRate=100, Use Spawn Probability=true)
  ]
  particleSpawnScript [
   &PartitionParticles(Number Of Partitions=4)
   &InitializeParticle(
    Lifetime Max=0.4
    Lifetime Min=0.2
    Lifetime Mode=::NewEnumerator1
    Sprite Size Max=<1 5>
    Sprite Size Min=<0.75 2>
    Sprite Size Mode=::NewEnumerator2
    UsePositionOffset=true
    Write Material Random=true
   )
   &SkeletalMeshLocation(
    Bone / Socket Index=$Particles.PartitionParticles.Partition
    MaskSpawnGroup=false
    Mesh Sampling Type=::NewEnumerator5
    Sampled Position Offset=<-5 0 0>
    Skeletal Mesh=$System.Skeletal Mesh
    Socket Sampling Mode=::NewEnumerator5
   )
   &CameraOffset(Camera Offset Amount=-5)
   &AddVelocity(
    Additional Quaternion Rotation=false
    Module.UseConeFalloff=false
    Module.UseVelDistribution=false
    Override Seed=false
    Rotation Mode=::NewEnumerator3
    Rotation Quaternion:Quat4f=$Particles.SkeletalMeshLocation.SampledRotation
    Velocity Mode=::NewEnumerator2
    Velocity Speed=&Multiply_Float(A=1250, B=&RandomRangeFloat(Minimum=0.3, Maximum=1))
   )
   /SetVariables(Particles.RandomLerp:NiagaraFloat=&MakeCustomFloatFromBool(Bool=&RandomBool(), True Float=0, False Float=1)) {isSetParameters=true}
  ]
  particleUpdateScript [
   &ParticleState()
   /SetVariables(Particles.Glow03_Color=&MakeLinearColorFromVectorAndFloat(
     "Vector (RGB)"=&MakeVectorFromLinearColorRGB(Color=&MultiplyLinearColor(
       Linear Color=&LerpLinearColors(StartColor=<rgba 1 0.0833 0 0.2>, EndColor=<rgba 0 0.25 1 0.2>, LerpFactor=$System.Timeline.NormalizedTime)
       Scale Factor=&MakeVector4_FromFloat(Float=0.75)
      ))
     "Float (Alpha)"=&Multiply_Float(A=0.2, B=1)
    )) {isSetParameters=true}
   /SetVariables(Particles.Glow03_Size=&Multiply_Vector2D_ByFloat(
     Float=&FloatFromCurve(
      FloatCurve=DataInterface() {propertyValues {Curve {keys [
          <at 0 0 RCIM_Cubic>
          <at 0.2 1.3 RCIM_Cubic>
          {arriveTangent=-0.1423, interpMode=RCIM_Cubic, leaveTangent=-0.1423, tangentMode=RCTM_User, time=1, value=0}
         ]}}}
      Scale Curve=50
     )
     Vector2D=<0.3 0.5>
    )) {isSetParameters=true}
   &ScaleColor(
    Linear Color Curve=DataInterface() {propertyValues {
      AlphaCurve {keys [<at 0 1>,<at 1 0>]}
      BlueCurve {keys [<at 0 20>,<at 0.1051 5.5649>,<at 0.1911 1>,<at 0.8847 0.2>]}
      GreenCurve {keys [<at 0 20>,<at 0.1051 5.5649>,<at 0.1911 1>,<at 0.8847 0.2>]}
      RedCurve {keys [<at 0 20>,<at 0.1051 5.5649>,<at 0.1911 1>,<at 0.8847 0.2>]}
     }}
    Scale Mode=::NewEnumerator2
   )
   &ScaleSpriteSize(Scale Factor=&Vector2DFromFloat(Value=&FloatFromCurve(
      FloatCurve=DataInterface() {propertyValues {Curve {keys [
          <at 0 0 RCIM_Cubic>
          {interpMode=RCIM_Cubic, leaveTangent=-3.0296, tangentMode=RCTM_Break, time=0.2, value=1.2}
          {arriveTangent=-2.8894, interpMode=RCIM_Cubic, leaveTangent=-2.8894, tangentMode=RCTM_User, time=1, value=0.1}
         ]}}}
      Scale Curve=&RandomRangeFloat(Minimum=0.5, Maximum=1.5)
     ))) {module=ScaleSpriteSize001}
   &Drag(Drag=22.5, Use Linear Drag=true, Use Rotational Drag=false)
   &SolveForcesAndVelocity(Clamp Velocity=false, Limit Acceleration=false)
   &ScaleColor(
    Scale RGB=&MakeVectorFromLinearColorRGB(Color=&LerpLinearColors(StartColor=<rgba 5 0.8867 0.1338 1>, EndColor=<rgba 0 1.25 5 1>, LerpFactor=$System.Timeline.NormalizedTime))
    ScaleA=true
    ScaleRGB=true
   ) {module=ScaleColor001}
  ]
  rendererClasses [&NiagaraSpriteRendererProperties]
  renderers [
   {rendererClass=&NiagaraSpriteRendererProperties}
   {rendererClass=&NiagaraSpriteRendererProperties, rendererIndex=1}
  ]
 }
}
summary {
 systemName=NS_SevenSlide_EngineThruster
 emitters [
  {emitterName=Flash, rendererClasses [&NiagaraSpriteRendererProperties]}
  {emitterName=Light, rendererClasses [&NiagaraSpriteRendererProperties, &NiagaraLightRendererProperties]}
  {emitterName=Reactor, rendererClasses [&NiagaraMeshRendererProperties]}
  {emitterName=FlameBig, rendererClasses [&NiagaraMeshRendererProperties]}
  {emitterName=Sparks001, simTarget=GPUComputeSim, rendererClasses [&NiagaraSpriteRendererProperties]}
 ]
 userVariables [
  User.Fuel {type=NiagaraFloat, defaultValue=1, description=""}
  User.IsSlomo {type=NiagaraBool}
 ]
}
