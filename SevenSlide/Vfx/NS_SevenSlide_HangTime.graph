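# Niagara system for the SevenSlide hang-time effect, attached to a target skeletal mesh and exposing a time-remaining parameter.
# Glow keeps a blue mesh on the skeleton with a camera offset, dimming its alpha over the system Timeline.
# SmallHexagons continuously spawns short-lived red or blue sprites around the mesh that inherit its movement and pulse via waveforms and curves.
# Trail uses custom AnimTrail modules to draw two tapering blue ribbons between bone/socket pairs.
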NiagaraSystem /SevenSlide/Vfx/NS_SevenSlide_HangTime
dependencies {
 usedDataInterfaces [
  &NiagaraDataInterfaceArrayInt32
  &NiagaraDataInterfaceCurve
  &NiagaraDataInterfaceArrayColor
  &NiagaraDataInterfaceSkeletalMesh
 ]
 usedDynamicInputs [&SetBoolByFloatComparison]
 usedModules [
  &self:SystemSpawnScript.SystemScriptSource.SystemScriptGraph.NiagaraNodeAssignment_0.SetVariables_#1
  &SystemState
  &Timeline
 ]
}
emitters {
 Glow {
  emitterUpdateScript [
   &EmitterState(
    Inactive Response=::NewEnumerator1
    Life Cycle Mode=::NewEnumerator1
    Loop Behavior=::NewEnumerator1
    Loop Duration Mode=::NewEnumerator1
    UseLoopDelay=false
   )
   &SpawnBurst_Instantaneous(Use Spawn Probability=false, UseLoopCountLimit=false)
  ]
  particleSpawnScript [&InitializeParticle(
    Color=<rgba 0 0.2087 1 1>
    Mesh Scale=&Multiply_VectorByFloat(Vector=<1.8 1.2 1.2>, Float=1)
    Mesh Scale Mode=::NewEnumerator4
    UsePositionOffset=false
    Write Material Random=true
   )]
  particleUpdateScript [
   &ParticleState(Kill Particles When Lifetime Has Elapsed=false, Let Infinitely Lived Particles Die When Emitter Deactivates=true)
   &SkeletalMeshLocation(Bone Sampling Mode=::NewEnumerator4, MaskSpawnGroup=false, Skeletal Mesh=$System.TargetSM)
   &CameraOffset(Camera Offset Amount=35)
   &ScaleColor(Scale Alpha=&Lerp_Float(A=2, B=0.75, Alpha=$System.Timeline.NormalizedTime), ScaleA=true, ScaleRGB=true)
  ]
  rendererClasses [&NiagaraMeshRendererProperties]
  renderers [{rendererClass=&NiagaraMeshRendererProperties, rendererIndex=0}]
 }
 SmallHexagons {
  emitterUpdateScript [
   &EmitterState(Life Cycle Mode=::NewEnumerator1, Loop Behavior=::NewEnumerator1, Loop Duration Mode=::NewEnumerator1, UseLoopDelay=false)
   &SpawnRate(SpawnRate=22, Use Spawn Probability=false)
  ]
  particleSpawnScript [
   &InitializeParticle(
    Color=&SelectLinearColorFromArray(
     Array Sampling Mode=::NewEnumerator0
     Direct Array Index=?
     Color Selection Array=DataInterface() {propertyValues {ColorData [<rgba 1 0 0 1>, <rgba 0 0 1 1>]}}
     Random Seed=0
    )
    Lifetime=&RandomRangeFloat(Minimum=0.6, Maximum=0.7)
    Sprite Rotation Mode=::NewEnumerator2
    Sprite Size Mode=::NewEnumerator3
    Uniform Sprite Size=&RandomRangeFloat(Minimum=2.5, Maximum=12.5)
    UsePositionOffset=false
    Write Material Random=true
   )
   &SkeletalMeshLocation(MaskSpawnGroup=false, OverrideRandomSeed=false, Skeletal Mesh=$System.TargetSM)
   &ShapeLocation(MaskSpawnGroup=false, Non Uniform Scale=<1 1 2>, Override Seed=false, Sphere Radius=7.5, UseOwnerScale=true)
  ]
  particleUpdateScript [
   &ParticleState()
   &Drag(Drag=25, Use Linear Drag=true, Use Rotational Drag=true)
   &InheritSourceMovement(Applied Position Scale=<0.8 0.8 0.8>, ApplyForce=false, ApplyPosition=true, ApplyPositionFalloff=false)
   &SolveForcesAndVelocity(Clamp Velocity=false, Limit Acceleration=false)
   &ScaleColor(
    Scale Alpha=&FloatFromCurve(
     FloatCurve=DataInterface() {propertyValues {Curve {
        defaultValue=0
        keys [
         {arriveTangent=9.0725, interpMode=RCIM_Cubic, leaveTangent=9.0725, tangentMode=RCTM_User, time=0, value=0}
         {arriveTangent=-2.5249, interpMode=RCIM_Cubic, leaveTangent=-2.5249, tangentMode=RCTM_User, time=0.1408, value=0.9847}
         {arriveTangent=-0.0405, interpMode=RCIM_Cubic, leaveTangent=-0.0405, tangentMode=RCTM_User, time=1, value=0}
        ]
       }}}
     Scale Curve=&RandomRangeFloat(Minimum=0.5, Maximum=1)
    )
    Scale RGB=&VectorFromFloat(Value=&Waveform(
      Global Input Position=&RandomRangeFloat(Minimum=0, Maximum=1)
      "Waveform [1]"=::NewEnumerator10
      "[1] Frequency"=3
      "[1] Amplitude Min/Max"=<0.25 1>
     ))
    ScaleA=true
    ScaleRGB=true
   )
   &ScaleSpriteSize(
    Uniform Curve Scale=&Waveform(
     Global Input Position=&RandomRangeFloat(Minimum=0, Maximum=1)
     "Waveform [1]"=::NewEnumerator10
     "[1] Frequency"=1
     "[1] Amplitude Min/Max"=<0.25 1.25>
    )
    Uniform Curve Sprite Scale=DataInterface() {propertyValues {Curve {keys [
        {arriveTangent=0.6194, interpMode=RCIM_Cubic, leaveTangent=0.6194, tangentMode=RCTM_User, time=0, value=0.5}
        {arriveTangent=2.4584, interpMode=RCIM_Cubic, leaveTangent=2.4584, time=0.1536, value=0.6098}
        {arriveTangent=0.4987, interpMode=RCIM_Cubic, leaveTangent=0.4987, time=0.2096, value=1.0152}
        {arriveTangent=-2.7608, interpMode=RCIM_Cubic, leaveTangent=-2.7608, time=0.3216, value=0.6936}
        {arriveTangent=0.5033, interpMode=RCIM_Cubic, leaveTangent=0.5033, time=0.3786, value=0.5485}
        {arriveTangent=2.8903, interpMode=RCIM_Cubic, leaveTangent=2.8903, time=0.4696, value=0.7681}
        {arriveTangent=-0.3148, interpMode=RCIM_Cubic, leaveTangent=-0.3148, time=0.5346, value=0.9995}
        {arriveTangent=0.0011, interpMode=RCIM_Cubic, leaveTangent=0.0011, time=0.6066, value=0.725}
        {arriveTangent=0.2153, interpMode=RCIM_Cubic, leaveTangent=0.2153, tangentMode=RCTM_User, time=1, value=1}
       ]}}}
   )
  ]
  rendererClasses [&NiagaraSpriteRendererProperties]
  renderers [{rendererClass=&NiagaraSpriteRendererProperties}]
 }
 Trail {
  emitterUpdateScript [
   &EmitterState(Life Cycle Mode=::NewEnumerator1, Loop Behavior=::NewEnumerator1, Loop Duration Mode=::NewEnumerator1, UseLoopDelay=false)
   &FN_SpawnAnimTrail(
    AnimTrailWidthMode=::FromSecond
    Number of Anim Trails=2
    Spawn Mode:ENiagaraSkeletalSpawning=::NewEnumerator0
    Switch on ENiagara_AnimTrail_Source:ENiagara_AnimTrail_Source=::NewEnumerator1
   )
  ]
  particleSpawnScript [
   &DynamicMaterialParameters(
    Index 0 Param 1=&RandomRangeFloat(Minimum=0, Maximum=0.1)
    Module.Index0Float0Write=true
    Module.Index0Float1Write=true
    Module.Index0Float2Write=true
    Module.Index0Float3Write=true
   )
   &FN_InitializeAnimTrailBoneIndecies(
    AnimTrail Selection:NiagaraInt32=&SelectIntFromArray(Direct Array Index=&ReturnExecIndex(), Int Selection Array=DataInterface() {propertyValues {IntData [0,1]}})
    First Bone or Socket Index:NiagaraInt32=&SelectIntFromArray(Direct Array Index=&ReturnExecIndex(), Int Selection Array=DataInterface() {propertyValues {IntData [5,6]}})
    Second Bone or Socket Index:NiagaraInt32=&SelectIntFromArray(Direct Array Index=&ReturnExecIndex(), Int Selection Array=DataInterface() {propertyValues {IntData [7,8]}})
   )
   &FN_InitializeAnimTrail(Lifetime=1.2, SkeletalMesh=$System.TargetSM)
   &Color(Color=<rgba 0 0.1667 1 0.35>)
  ]
  particleUpdateScript [
   &ParticleState()
   &FN_UpdateAnimTrail(SkeletalMesh=$System.TargetSM)
   /SetVariables(Particles.CustomVector:Vector3f=<0 0 1>) {isSetParameters=true}
   /SetVariables(Particles.RibbonWidth=&FloatFromCurve(
     FloatCurve=DataInterface() {propertyValues {Curve {
        defaultValue=5
        keys [<at 0 1 RCIM_Cubic>, {arriveTangent=-2.8259, interpMode=RCIM_Cubic, leaveTangent=-2.8259, tangentMode=RCTM_User, time=1, value=0}]
       }}}
     Scale Curve=25
    )) {isSetParameters=true}
  ]
  rendererClasses [&NiagaraRibbonRendererProperties]
  renderers [{rendererClass=&NiagaraRibbonRendererProperties}]
 }
}
summary {
 systemName=NS_SevenSlide_HangTime
 emitters [
  {emitterName=Glow, rendererClasses [&NiagaraMeshRendererProperties]}
  {emitterName=Trail, rendererClasses [&NiagaraRibbonRendererProperties]}
  {emitterName=SmallHexagons, rendererClasses [&NiagaraSpriteRendererProperties]}
 ]
 userVariables [
  User.HangtimeTimeRemaining {type=NiagaraFloat, defaultValue=0, description=""}
  User.RibbonIDOffset {type=NiagaraInt32}
 ]
}
